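# Sound cue for the Core Pistol's aim-down-sights entry, played as weapon foley with weapon reload attenuation.
# A random node picks with equal weight between the second ADS-start wave alone and a mix of both waves.
# In the mix, the first wave is pitched up and slightly delayed, layered with a modulated copy of the second.
# A root modulator adds slight upward pitch variation to either result.
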SoundCue /FlipperGameplay/Items/Weapons/CorePistol/Audio/Core_Pistol_ADS_IN_Cue
root=Modulator_0
nodes [
 {id=Delay_0, children [Modulator_2], properties {delayMax=0.05}}
 {id=Mixer_0, children [Delay_0,Modulator_1], properties {inputVolume [0.7,0.9]}}
 {id=Modulator_0, children [Random_0], properties {pitchMax=1.1, pitchMin=1}}
 {id=Modulator_1, children [WavePlayer_2]}
 {id=Modulator_2, children [WavePlayer_1], properties {pitchMax=1.2, pitchMin=1.1}}
 {id=Random_0, children [Mixer_0,WavePlayer_2], properties {weights [1,1]}}
 {id=WavePlayer_1, properties {soundWaveAssetPtr=&/Pistol_Core_ADS_Start_01}}
 {id=WavePlayer_2, properties {soundWaveAssetPtr=&/Pistol_Core_ADS_Start_02}}
]
properties {
 attenuationSettings=&Fort_Weapon_Reload_Att
 maxDistance=2250
 soundClassObject=&Fort_WeaponFoley
 volumeMultiplier=0.376
}
