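# Sound cue for the Core Pistol's aim-down-sights exit, played as weapon foley.
# A Random node picks between two equally weighted ADS end waves and feeds a Modulator at default settings for slight variation.
# The cue plays at low volume in the weapon foley sound class and borrows the shared weapon reload attenuation.
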SoundCue /FlipperGameplay/Items/Weapons/CorePistol/Audio/Core_Pistol_ADS_OUT_Cue
root=Modulator_0
nodes [
 {id=Modulator_0, children [Random_1]}
 {id=Random_1, children [WavePlayer_0,WavePlayer_3], properties {weights [1,1]}}
 {id=WavePlayer_0, properties {soundWaveAssetPtr=&/Pistol_Core_ADS_End_01}}
 {id=WavePlayer_3, properties {soundWaveAssetPtr=&/Pistol_Core_ADS_End_02}}
]
properties {
 attenuationSettings=&Fort_Weapon_Reload_Att
 maxDistance=2250
 soundClassObject=&Fort_WeaponFoley
 volumeMultiplier=0.1
}
