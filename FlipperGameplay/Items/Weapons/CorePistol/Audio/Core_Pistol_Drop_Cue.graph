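# Sound cue for the Core Pistol drop sound in the Flipper gameplay item set.
# A Random node picks between two equally weighted drop waves and feeds a Modulator that slightly varies pitch on each play.
# The cue uses the shared weapon-drop attenuation and concurrency assets, is routed to the loot-collect sound class, and plays at slightly reduced volume.
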SoundCue /FlipperGameplay/Items/Weapons/CorePistol/Audio/Core_Pistol_Drop_Cue
root=Modulator_0
nodes [
 {id=Modulator_0, children [Random_1], properties {pitchMax=1.1, pitchMin=0.9}}
 {id=Random_1, children [WavePlayer_0,WavePlayer_1], properties {weights [1,1]}}
 {id=WavePlayer_0, properties {soundWaveAssetPtr=&/Pistol_Core_Drop_01}}
 {id=WavePlayer_1, properties {soundWaveAssetPtr=&/Pistol_Core_Drop_02}}
]
properties {
 attenuationSettings=&Fort_Weapon_Drop_Att
 maxDistance=3000
 soundClassObject=&Fort_Loot_Collect
 volumeMultiplier=0.9
 concurrencyOverrides {resolutionRule=PreventNew}
 concurrencySet [&Weapon_Drop_StopOldest]
}
