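# Sound cue for the first-person firing sound of the Core Pistol, routed through a first-person weapon fire sound class and attenuation.
# A mixer layers a random close-fire shot with slight upward pitch variation over a quieter, delayed random shell-casing drop.
# The casing layer borrows burst-fire SMG pistol samples and randomizes to lower pitch and varied volume.
# It references a first-person gunfire concurrency set and carries per-owner, stop-oldest concurrency overrides.
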SoundCue /FlipperGameplay/Items/Weapons/CorePistol/Audio/Core_Pistol_Fire_1P_Cue
root=Mixer_0
nodes [
 {id=Delay_0, children [Modulator_3], properties {delayMax=0.4, delayMin=0.4}}
 {id=Mixer_0, children [Modulator_2,Delay_0], properties {inputVolume [1,0.2]}}
 {id=Modulator_2, children [Random_0], properties {pitchMax=1.1, pitchMin=1}}
 {id=Modulator_3, children [Random_5], properties {pitchMax=0.95, pitchMin=0.85, volumeMax=1, volumeMin=0.7}}
 {id=Random_0, children [WavePlayer_0,WavePlayer_1,WavePlayer_2,WavePlayer_3,WavePlayer_4,WavePlayer_5], properties {weights [1,1,1,1,1,1]}}
 {id=Random_5, children [WavePlayer_16, WavePlayer_17, WavePlayer_18, WavePlayer_19, WavePlayer_20], properties {weights [1,1,1,1,1]}}
 {id=WavePlayer_0, properties {soundWaveAssetPtr=&/Pistol_Core_Fire_Close_01}}
 {id=WavePlayer_1, properties {soundWaveAssetPtr=&/Pistol_Core_Fire_Close_02}}
 {id=WavePlayer_16, properties {soundWaveAssetPtr=/Game/Athena/Sounds/Weapons/Pistol_BurstFireSMG/Pistol_BurstFireSMG_Bullet_Casings_Drop_01}}
 {id=WavePlayer_17, properties {soundWaveAssetPtr=/Game/Athena/Sounds/Weapons/Pistol_BurstFireSMG/Pistol_BurstFireSMG_Bullet_Casings_Drop_02}}
 {id=WavePlayer_18, properties {soundWaveAssetPtr=/Game/Athena/Sounds/Weapons/Pistol_BurstFireSMG/Pistol_BurstFireSMG_Bullet_Casings_Drop_03}}
 {id=WavePlayer_19, properties {soundWaveAssetPtr=/Game/Athena/Sounds/Weapons/Pistol_BurstFireSMG/Pistol_BurstFireSMG_Bullet_Casings_Drop_04}}
 {id=WavePlayer_2, properties {soundWaveAssetPtr=&/Pistol_Core_Fire_Close_03}}
 {id=WavePlayer_20, properties {soundWaveAssetPtr=/Game/Athena/Sounds/Weapons/Pistol_BurstFireSMG/Pistol_BurstFireSMG_Bullet_Casings_Drop_05}}
 {id=WavePlayer_3, properties {soundWaveAssetPtr=&/Pistol_Core_Fire_Close_04}}
 {id=WavePlayer_4, properties {soundWaveAssetPtr=&/Pistol_Core_Fire_Close_05}}
 {id=WavePlayer_5, properties {soundWaveAssetPtr=&/Pistol_Core_Fire_Close_06}}
]
properties {
 attenuationSettings=&Fort_Weapon_Projectile_FirstPerson_Athena_Att
 bDebug=true
 maxDistance=2097152
 soundClassObject=&Fort_WeaponRanged_Fire_1P_PSM
 volumeMultiplier=0.8
 attenuationOverrides {manualReverbSendLevel=-1.0165, reverbSendMethod=Manual}
 concurrencyOverrides {bLimitToOwner=true, maxCount=2, resolutionRule=StopOldest, volumeScale=0.9}
 concurrencySet [&Weapon_Gunfire_FirstPerson_Conc]
}
