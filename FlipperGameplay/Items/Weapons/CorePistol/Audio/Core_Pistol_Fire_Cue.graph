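# Sound cue for Core Pistol gunfire, layering randomized close, medium and distant shot recordings.
# Each layer has its own attenuation, with a third-person fire class for close and a distant-fire class for the others.
# Close and medium layers vary pitch, medium and distant boost volume, and nested mixers balance the three, keeping distant low.
# Cue-level defaults use first-person sound class, attenuation and gunfire concurrency, and the debug flag is on.
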SoundCue /FlipperGameplay/Items/Weapons/CorePistol/Audio/Core_Pistol_Fire_Cue
root=Mixer_2
nodes [
 {id=Attenuation_3, children [Modulator_1], properties {attenuationSettings=&Fort_Weapon_Projectile_Close_Athena_Att}}
 {
  id=Attenuation_4
  children [Modulator_0]
  properties {attenuationSettings=&Fort_Weapon_Projectile_Distant_Mono_Athena_Att}
 }
 {
  id=Attenuation_5
  children [Modulator_2]
  properties {attenuationSettings=&Fort_Weapon_Projectile_Medium_Mono_Athena_Att}
 }
 {id=Mixer_2, children [Modulator_3,SoundClass_4], properties {inputVolume [1,0.1]}}
 {id=Mixer_3, children [SoundClass_3,SoundClass_5], properties {inputVolume [0.9,0.6]}}
 {id=Modulator_0, children [Random_1], properties {volumeMax=2, volumeMin=2}}
 {id=Modulator_1, children [Random_0], properties {pitchMax=1.1, pitchMin=1}}
 {id=Modulator_2, children [Random_4], properties {pitchMax=1.1, pitchMin=1, volumeMax=2, volumeMin=2}}
 {id=Modulator_3, children [Mixer_3]}
 {id=Random_0, children [WavePlayer_0,WavePlayer_1,WavePlayer_2,WavePlayer_3,WavePlayer_4,WavePlayer_5], properties {weights [1,1,1,1,1,1]}}
 {id=Random_1, children [WavePlayer_14, WavePlayer_15, WavePlayer_16], properties {weights [1,1,1]}}
 {id=Random_4, children [WavePlayer_6,WavePlayer_7,WavePlayer_8], properties {weights [1,1,1]}}
 {
  id=SoundClass_3
  children [Attenuation_3]
  properties {soundClassOverride=&Fort_WeaponRanged_Fire_3P_PSM}
 }
 {
  id=SoundClass_4
  children [Attenuation_4]
  properties {soundClassOverride=&Fort_WeaponRanged_Fire_Distant}
 }
 {
  id=SoundClass_5
  children [Attenuation_5]
  properties {soundClassOverride=&Fort_WeaponRanged_Fire_Distant}
 }
 {id=WavePlayer_0, properties {soundWaveAssetPtr=&/Pistol_Core_Fire_Close_01}}
 {id=WavePlayer_1, properties {soundWaveAssetPtr=&/Pistol_Core_Fire_Close_02}}
 {id=WavePlayer_14, properties {soundWaveAssetPtr=&/Pistol_Core_Fire_Distant_04}}
 {id=WavePlayer_15, properties {soundWaveAssetPtr=&/Pistol_Core_Fire_Distant_05}}
 {id=WavePlayer_16, properties {soundWaveAssetPtr=&/Pistol_Core_Fire_Distant_06}}
 {id=WavePlayer_2, properties {soundWaveAssetPtr=&/Pistol_Core_Fire_Close_03}}
 {id=WavePlayer_3, properties {soundWaveAssetPtr=&/Pistol_Core_Fire_Close_04}}
 {id=WavePlayer_4, properties {soundWaveAssetPtr=&/Pistol_Core_Fire_Close_05}}
 {id=WavePlayer_5, properties {soundWaveAssetPtr=&/Pistol_Core_Fire_Close_06}}
 {id=WavePlayer_6, properties {soundWaveAssetPtr=&/Pistol_Core_Fire_Medium_01}}
 {id=WavePlayer_7, properties {soundWaveAssetPtr=&/Pistol_Core_Fire_Medium_02}}
 {id=WavePlayer_8, properties {soundWaveAssetPtr=&/Pistol_Core_Fire_Medium_03}}
]
properties {
 attenuationSettings=&Fort_Weapon_Projectile_FirstPerson_Athena_Att
 bDebug=true
 maxDistance=2097152
 soundClassObject=&Fort_WeaponRanged_Fire_1P_PSM
 volumeMultiplier=0.9
 concurrencyOverrides {bLimitToOwner=true, maxCount=2, resolutionRule=StopOldest}
 concurrencySet [&Weapon_Gunfire_FirstPerson_Conc]
}
