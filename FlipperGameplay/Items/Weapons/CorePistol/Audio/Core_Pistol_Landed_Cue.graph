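# Sound cue that plays the landing sound for a Core Pistol item in the loot-collect sound class, likely when it drops to the ground.
# A Random node picks evenly between two landed wave variations, and a Modulator lowers the pitch slightly with random variation.
# It sets both a basic foley attenuation asset and an override attenuation with short falloff, and its own concurrency rule caps simultaneous plays, blocking new ones past the limit.
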SoundCue /FlipperGameplay/Items/Weapons/CorePistol/Audio/Core_Pistol_Landed_Cue
root=Modulator_1
nodes [
 {id=Modulator_1, children [Random_1], properties {pitchMax=0.95, pitchMin=0.85}}
 {id=Random_1, children [WavePlayer_0,WavePlayer_4], properties {weights [1,1]}}
 {id=WavePlayer_0, properties {soundWaveAssetPtr=&/Pistol_Core_Landed_01}}
 {id=WavePlayer_4, properties {soundWaveAssetPtr=&/Pistol_Core_Landed_02}}
]
properties {
 attenuationSettings=&Foley_Attenuation_Basic_Att
 bOverrideAttenuation=true
 bOverrideConcurrency=true
 maxDistance=1000
 soundClassObject=&Fort_Loot_Collect
 volumeMultiplier=1.1
 attenuationOverrides {falloffDistance=1000, attenuationShapeExtents {x=0}}
 concurrencyOverrides {maxCount=3, resolutionRule=PreventNew}
}
