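# Sound cue for the Core Pistol's low-ammo weapon audio.
# A Random node picks one of five equally weighted low-ammo wave variations per play.
# It uses a close weapon projectile attenuation, a degraded low-ammo ranged weapon sound class and a boosted volume.
# A concurrency override prevents new instances from starting once the limit is reached.
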SoundCue /FlipperGameplay/Items/Weapons/CorePistol/Audio/Core_Pistol_Low_Ammo_Cue
root=Random_0
nodes [
 {id=Random_0, children [WavePlayer_0,WavePlayer_2,WavePlayer_3,WavePlayer_4,WavePlayer_6], properties {weights [1,1,1,1,1]}}
 {id=WavePlayer_0, properties {soundWaveAssetPtr=&/Pistol_Core_LowAmmo_01}}
 {id=WavePlayer_2, properties {soundWaveAssetPtr=&/Pistol_Core_LowAmmo_02}}
 {id=WavePlayer_3, properties {soundWaveAssetPtr=&/Pistol_Core_LowAmmo_03}}
 {id=WavePlayer_4, properties {soundWaveAssetPtr=&/Pistol_Core_LowAmmo_04}}
 {id=WavePlayer_6, properties {soundWaveAssetPtr=&/Pistol_Core_LowAmmo_06}}
]
properties {
 attenuationSettings=&Fort_Weapon_Projectile_Close_Athena_Att
 bOverrideConcurrency=true
 maxDistance=4950
 soundClassObject=/Game/Sounds/FortSoundClassesAndModes/Fort_WeaponRanged_LowAmmo_Degraded_PSM
 volumeMultiplier=2.15
 concurrencyOverrides {resolutionRule=PreventNew}
}
