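# Sound cue for the Core Pistol's out-of-ammo sound.
# A Random node picks one of three equally weighted wave variants and feeds a Modulator that pins pitch at 1.
# Playback uses a shared close-range weapon projectile attenuation, a first-person ranged-fire sound class and first-person gunfire concurrency, with a boosted volume multiplier.
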SoundCue /FlipperGameplay/Items/Weapons/CorePistol/Audio/Core_Pistol_OutofAmmo_Cue
root=Modulator_0
nodes [
 {id=Modulator_0, children [Random_1], properties {pitchMax=1, pitchMin=1}}
 {id=Random_1, children [WavePlayer_4,WavePlayer_5,WavePlayer_6], properties {weights [1,1,1]}}
 {id=WavePlayer_4, properties {soundWaveAssetPtr=&/Pistol_Core_OutOfAmmo_01}}
 {id=WavePlayer_5, properties {soundWaveAssetPtr=&/Pistol_Core_OutOfAmmo_02}}
 {id=WavePlayer_6, properties {soundWaveAssetPtr=&/Pistol_Core_OutOfAmmo_03}}
]
properties {
 attenuationSettings=&Fort_Weapon_Projectile_Close_Athena_Att
 maxDistance=4950
 soundClassObject=&Fort_WeaponRanged_Fire_1P_PSM
 volumeMultiplier=1.3
 attenuationOverrides {falloffDistance=1600}
 concurrencyOverrides {maxCount=2, resolutionRule=StopOldest}
 concurrencySet [&Weapon_Gunfire_FirstPerson_Conc]
}
