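# Sound cue for picking up the Core Pistol, classed as loot-collect audio.
# A Random node picks one of two equally weighted pickup waves and feeds a Modulator that varies the playback slightly.
# It plays at reduced volume through the shared weapon pickup attenuation and uses the PickupMaximum_Oldest concurrency set to limit overlapping pickup sounds.
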SoundCue /FlipperGameplay/Items/Weapons/CorePistol/Audio/Core_Pistol_Pickup_Cue
root=Modulator_0
nodes [
 {id=Modulator_0, children [Random_0], properties {pitchMin=1}}
 {id=Random_0, children [WavePlayer_0,WavePlayer_2], properties {weights [1,1]}}
 {id=WavePlayer_0, properties {soundWaveAssetPtr=&/Pistol_Core_Pickup_02}}
 {id=WavePlayer_2, properties {soundWaveAssetPtr=&/Pistol_Core_Pickup_03}}
]
properties {
 attenuationSettings=&Fort_Weapon_Pickup_Att
 maxDistance=3500
 soundClassObject=&Fort_Loot_Collect
 volumeMultiplier=0.5538
 concurrencyOverrides {maxCount=1, resolutionRule=PreventNew}
 concurrencySet [&PickupMaximum_Oldest]
}
