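# Sound cue for the cartridge-insert foley during a Core Pistol reload.
# A single wave player feeds the reload-insert recording through a modulator at the root for pitch and volume variation.
# It plays quietly in the weapon foley sound class with reload attenuation, overrides concurrency, and has its debug flag left on.
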SoundCue /FlipperGameplay/Items/Weapons/CorePistol/Audio/Core_Pistol_Reload_Cartridge_Cue
root=Modulator_0
nodes [
 {id=Modulator_0, children [WavePlayer_0]}
 {id=WavePlayer_0, properties {soundWaveAssetPtr=&/Pistol_Core_ReloadInsert_01}}
]
properties {
 attenuationSettings=&Fort_Weapon_Reload_Att
 bDebug=true
 bOverrideConcurrency=true
 maxDistance=2250
 soundClassObject=&Fort_WeaponFoley
 volumeMultiplier=0.251
}
