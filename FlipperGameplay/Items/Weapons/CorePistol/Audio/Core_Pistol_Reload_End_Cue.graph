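# Sound cue for the end of the Core Pistol reload, classed as weapon foley and using the shared weapon reload attenuation.
# A mixer layers one of two randomly chosen reload-end waves with a slightly quieter equip wave that has its own pitch and volume modulation.
# A root modulator varies the combined mix, with pitch capped at normal.
# The cue plays at reduced volume, overrides concurrency and has its debug flag on.
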SoundCue /FlipperGameplay/Items/Weapons/CorePistol/Audio/Core_Pistol_Reload_End_Cue
root=Modulator_0
nodes [
 {id=Mixer_0, children [Random_0,Modulator_1], properties {inputVolume [1,0.9]}}
 {id=Modulator_0, children [Mixer_0], properties {pitchMax=1}}
 {id=Modulator_1, children [WavePlayer_5], properties {pitchMin=0.9, volumeMin=0.6}}
 {id=Random_0, children [WavePlayer_0,WavePlayer_1], properties {weights [1,1]}}
 {id=WavePlayer_0, properties {soundWaveAssetPtr=&/Pistol_Core_ReloadEnd_01}}
 {id=WavePlayer_1, properties {soundWaveAssetPtr=&/Pistol_Core_ReloadEnd_02}}
 {id=WavePlayer_5, properties {soundWaveAssetPtr=&/Pistol_Core_Equip_03}}
]
properties {
 attenuationSettings=&Fort_Weapon_Reload_Att
 bDebug=true
 bOverrideConcurrency=true
 maxDistance=2250
 soundClassObject=&Fort_WeaponFoley
 volumeMultiplier=0.2
}
