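# Sound cue for the Core Pistol's reload start, played as weapon foley with reload attenuation.
# A mixer layers the main reload-start wave with a slightly delayed, pitched-down random pick of two reused ADS-start waves.
# A root modulator at default settings wraps the whole mix.
# The cue lowers overall volume, overrides concurrency and has its debug flag left on.
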SoundCue /FlipperGameplay/Items/Weapons/CorePistol/Audio/Core_Pistol_Reload_Start_Cue
root=Modulator_0
nodes [
 {id=Delay_0, children [Modulator_1], properties {delayMax=0.2, delayMin=0.15}}
 {id=Mixer_0, children [WavePlayer_0,Delay_0], properties {inputVolume [1,1]}}
 {id=Modulator_0, children [Mixer_0]}
 {id=Modulator_1, children [Random_0], properties {pitchMax=0.9, pitchMin=0.85}}
 {id=Random_0, children [WavePlayer_1,WavePlayer_2], properties {weights [1,1]}}
 {id=WavePlayer_0, properties {soundWaveAssetPtr=&/Pistol_Core_ReloadStart_01}}
 {id=WavePlayer_1, properties {soundWaveAssetPtr=&/Pistol_Core_ADS_Start_01}}
 {id=WavePlayer_2, properties {soundWaveAssetPtr=&/Pistol_Core_ADS_Start_02}}
]
properties {
 attenuationSettings=&Fort_Weapon_Reload_Att
 bDebug=true
 bOverrideConcurrency=true
 maxDistance=2250
 soundClassObject=&Fort_WeaponFoley
 volumeMultiplier=0.326
}
